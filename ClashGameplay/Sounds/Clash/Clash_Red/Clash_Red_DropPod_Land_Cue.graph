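# Sound cue for the landing of the red Clash drop pod, layering three sounds on the Fort_Gadgets sound class.
# A mixer blends a distant supply-drop landing, an envelope-shortened close landing with low-pass distance attenuation, and a slightly quieter, modulated pod spawn sound.
# The spawn layer overrides attenuation with a very large range so it carries across long distances.
# Concurrency allows one instance, stopping the farthest first.
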SoundCue /ClashGameplay/Sounds/Clash/Clash_Red/Clash_Red_DropPod_Land_Cue
root=Mixer_1
nodes [
 {
  id=Attenuation_1
  children [Modulator_1]
  properties {bOverrideAttenuation=true, attenuationOverrides {falloffDistance=500000, stereoSpread=5000, attenuationShapeExtents {x=250000}}}
 }
 {id=Attenuation_2, children [SoundClass_4], properties {attenuationSettings=&Fort_GadgetActivate_Att}}
 {
  id=Attenuation_3
  children [SoundClass_3]
  properties {
   attenuationSettings=&Fort_GadgetActivate_Att
   bOverrideAttenuation=true
   attenuationOverrides {bAttenuateWithLPF=true, falloffDistance=2500, lPFFrequencyAtMax=1000, lPFRadiusMax=2500, lPFRadiusMin=400, stereoSpread=1000}
  }
 }
 {
  id=Enveloper_1
  children [WavePlayer_3]
  properties {
   pitchCurve {editorCurveData {keys [{time=0}]}}
   volumeCurve {editorCurveData {keys [<at 0 1>,<at 0.2649 1.0002>,<at 0.2968 0.9159>,<at 0.3107 0.7872>,<at 0.3311 0.4569>,<at 0.5427 -0.007>]}}
  }
 }
 {id=Mixer_1, children [Attenuation_2, Attenuation_3, SoundClass_5], properties {inputVolume [1,1,0.7]}}
 {id=Modulator_1, children [WavePlayer_5]}
 {id=SoundClass_3, children [Enveloper_1], properties {soundClassOverride=&Fort_Gadgets}}
 {id=SoundClass_4, children [WavePlayer_2], properties {soundClassOverride=&Fort_Gadgets}}
 {id=SoundClass_5, children [Attenuation_1], properties {soundClassOverride=&Fort_Gadgets}}
 {id=WavePlayer_2, properties {soundWaveAssetPtr=/Game/Sounds/Foley_Loot/Containers/SupplyDrops/sfx_supplydrop_land_distant}}
 {id=WavePlayer_3, properties {soundWaveAssetPtr=/Game/Sounds/Fort_Gadgets/Supply_Drop/supply_drop_land3}}
 {id=WavePlayer_5, properties {soundWaveAssetPtr=&/Clash_Black_Red_Pod_Spawn_01}}
]
properties {
 bOverrideConcurrency=true
 maxDistance=750000
 volumeMultiplier=2
 concurrencyOverrides {maxCount=1, resolutionRule=StopFarthestThenPreventNew}
}
