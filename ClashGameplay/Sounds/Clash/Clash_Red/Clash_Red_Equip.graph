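# Sound cue for the Clash Red equip sound in ClashGameplay.
# A single wave player feeds the root with one equip wave, played at reduced volume through the loot collect sound class.
# It borrows the close umbrella weapon attenuation for spatial falloff.
# A concurrency override allows only one instance at a time and drops new requests while it plays.
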SoundCue /ClashGameplay/Sounds/Clash/Clash_Red/Clash_Red_Equip
root=WavePlayer_1
nodes [{id=WavePlayer_1, properties {soundWaveAssetPtr=&/Clash_Red_Equip_01}}]
properties {
 attenuationSettings=/Game/Sounds/Attenuation/Fort_Weapon_Umbrella_Close_Athena_Att
 bOverrideConcurrency=true
 maxDistance=4500
 soundClassObject=&Fort_Loot_Collect
 volumeMultiplier=0.597
 concurrencyOverrides {maxCount=1, resolutionRule=PreventNew}
}
