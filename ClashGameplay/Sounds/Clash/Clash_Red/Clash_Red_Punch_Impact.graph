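# Sound Cue for the Clash Red punch impact, routed to the melee weapon sound class.
# A mixer layers a random pick of four equally weighted Black/Red hit waves with shared Vertigo pullback and impact whoosh waves.
# The root modulator adds no pitch or volume variation, and the cue plays at reduced volume with close-range umbrella weapon attenuation.
# A concurrency override allows one instance at a time and blocks new ones.
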SoundCue /ClashGameplay/Sounds/Clash/Clash_Red/Clash_Red_Punch_Impact
root=Modulator_0
nodes [
 {id=Mixer_0, children [Random_0,WavePlayer_5,WavePlayer_7], properties {inputVolume [1,1,1]}}
 {id=Modulator_0, children [Mixer_0], properties {pitchMax=1, pitchMin=1, volumeMax=1, volumeMin=1}}
 {id=Random_0, children [WavePlayer_0,WavePlayer_1,WavePlayer_2,WavePlayer_3], properties {weights [1,1,1,1]}}
 {id=WavePlayer_0, properties {soundWaveAssetPtr=&/Clash_Black_Red_Hit_04}}
 {id=WavePlayer_1, properties {soundWaveAssetPtr=&/Clash_Black_Red_Hit_01}}
 {id=WavePlayer_2, properties {soundWaveAssetPtr=&/Clash_Black_Red_Hit_02}}
 {id=WavePlayer_3, properties {soundWaveAssetPtr=&/Clash_Black_Red_Hit_03}}
 {id=WavePlayer_5, properties {soundWaveAssetPtr=/ClashGameplay/Gameplay/Clash/Shared/Audio/Vertigo_SNG_Pullback_v04}}
 {id=WavePlayer_7, properties {soundWaveAssetPtr=/ClashGameplay/Gameplay/Clash/Shared/Audio/Vertigo_SNG_Impact_Whoosh_01}}
]
properties {
 attenuationSettings=/Game/Sounds/Attenuation/Fort_Weapon_Umbrella_Close_Athena_Att
 bOverrideConcurrency=true
 maxDistance=4500
 soundClassObject=&Fort_WeaponMelee
 volumeMultiplier=0.5
 concurrencyOverrides {maxCount=1, resolutionRule=PreventNew}
}
